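# M_InteractionCylinder_Translucent is a translucent Default Lit material, apparently a transparency test for an interaction cylinder among the cosmetic shoe assets.
# A light blue Color parameter is multiplied by a Fresnel term to tint grazing angles, alongside a scalar Opacity parameter that defaults to half.
# Usage flags cover skeletal meshes, mesh deformers and Nanite.
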Material /CosmeticShoes/Assets/FORT_Shoes/Transparency_Test/M_InteractionCylinder_Translucent
graph {
 properties {
  bUsedWithMeshDeformer=true
  bUsedWithNanite=true
  bUsedWithSkeletalMesh=true
  blendMode=BLEND_Translucent
  pixelDepthOffsetMode=PDOM_Legacy
  usedShadingModels=Default Lit
 }
 nodes [
  Multiply(A=VectorParameter() {parameterName=Color, defaultValue=<rgba 0.4084 0.6219 0.9427 0>}.RGB, B=Fresnel() {exponent=2})
  ScalarParameter() {defaultValue=0.5, parameterName=Opacity}
 ]
}
