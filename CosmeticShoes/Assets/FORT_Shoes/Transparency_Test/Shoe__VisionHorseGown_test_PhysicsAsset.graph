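# Physics asset for the VisionHorseGown test shoe, kept in a transparency test folder of the cosmetic shoe assets.
# It gives the Object002, RootNode and Shoe_LOD0 bones one capsule each, with only a tiny capsule on RootNode.
# Two constraints tie Object002 and Shoe_LOD0 separately to RootNode, both using the same symmetric swing and twist limits.
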PhysicsAsset /CosmeticShoes/Assets/FORT_Shoes/Transparency_Test/Shoe__VisionHorseGown_test_PhysicsAsset
bodies {
 Object002 [{shapeType=Capsule, length=21.0308, radius=10.5329, center=<0 5.0133 6.8335>, rotation=<rot 0 -90 -90>}]
 RootNode [{shapeType=Capsule, length=0.505, radius=0.505, rotation=<rot 0 -90 0>}]
 Shoe_LOD0 [{shapeType=Capsule, length=21.6449, radius=14.5855, center=<0.0551 1.765 7.2249>, rotation=<rot 0 90 -89.9999>}]
}
constraints [
 {bone1Name=Shoe_LOD0, bone2Name=RootNode, swing1LimitDegrees=45, swing2LimitDegrees=45, twistLimitDegrees=45}
 {bone1Name=Object002, bone2Name=RootNode, swing1LimitDegrees=45, swing2LimitDegrees=45, twistLimitDegrees=45}
]
